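; Parent Gameplay Ability for Jump Boots, providing chained jumps with per-jump parameters from a JumpParams array indexed by JumpCounter.
; Jump-press events check blocking tags and ceiling clearance, then launch the player using scaled acceleration, bypassing bunny-hop prevention, and apply that jump's effects and cues.
; Landing applies land effects and cues, then either commits cooldown and ends after the final jump or starts a chain-reset timer.
; Other trigger events reset the counter.
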(event Custom|JumpBoots_OnJump
  (bind (_jumpparams_ceilingheighttotrace_42_97759bbf46c015fd67dbaa9a9fd52bdc _jumpparams_launch_xyvelocitymultiplier_28_fd149de643cf2bf44ec729bb1b36d832 _jumpparams_launch_zvelocity_27_5bbf517844fc06657a6325b36dc3ade8 _jumpparams_launch_overridexyvelocity_30_64dc724b4d2faf89d63e1b81035d0ff8 _jumpparams_onjump_gameplayeffecttoapply_43_70fc651342557968aa109eba1d306a59) (Utilities|Struct|BreakSTRUCTJumpBootsJumpParams (CallFunction|GetJumpParamsforCurrentJump)))
  (bind (_jumpparams_ceilingheighttotrace_42_97759bbf46c015fd67dbaa9a9fd52bdc_2 _jumpparams_launch_xyvelocitymultiplier_28_fd149de643cf2bf44ec729bb1b36d832_2 _jumpparams_launch_zvelocity_27_5bbf517844fc06657a6325b36dc3ade8_2 _jumpparams_launch_overridexyvelocity_30_64dc724b4d2faf89d63e1b81035d0ff8_2 _jumpparams_onjump_gameplayeffecttoapply_43_70fc651342557968aa109eba1d306a59_2 _jumpparams_onland_gameplayeffecttoapply_44_32a934db4255e53df4aaa78e2eb62d38 _jumpparams_onjump_gameplaycue_48_0fec94334bbbb7b0c7fa2b930cd62862) (Utilities|Struct|BreakSTRUCTJumpBootsJumpParams (CallFunction|GetJumpParamsforCurrentJump)))
  (CallFunction|UpdateJumpVars)
  (CallFunction|LaunchPlayer) ; Override jump logic and launch player. \nThis bypasses bunny hop prevention code.\n\nLaunch criteria is driven by JumpParams variable.
  (CallFunction|ValidateandExecuteGE _jumpparams_onjump_gameplayeffecttoapply_43_70fc651342557968aa109eba1d306a59) ; Add GE's to make each jump unique.
  (CallFunction|ValidateandExecuteGCN _jumpparams_onjump_gameplaycue_48_0fec94334bbbb7b0c7fa2b930cd62862) ; Add GCN/GCNL's  in blueprint to prevent\nreplication delay.
  (Networking|ForceNetUpdate (Variables|Hidden|GetPlayerRef)))

(event Ability|EventActivateAbilityFromEvent
  (params (EventData_EventTag :type Gameplay%20Tag%20Structure) (EventData_Instigator :type Actor%20Object%20Reference) (EventData_Target :type Actor%20Object%20Reference) (EventData_OptionalObject :type Object%20Reference) (EventData_OptionalObject2 :type Object%20Reference) (EventData_ContextHandle :type Gameplay%20Effect%20Context%20Handle%20Structure) (EventData_InstigatorTags :type Gameplay%20Tag%20Container%20Structure) (EventData_TargetTags :type Gameplay%20Tag%20Container%20Structure) (EventData_EventMagnitude :type Float%20%28single-precision%29) (EventData_TargetData :type Gameplay%20Ability%20Target%20Data%20Handle%20Structure) (EventData_InstancedEventData :type Instanced%20Struct%20Structure))
  (switch Utilities|FlowControl|Switch|SwitchonGameplayTag EventData_EventTag
    (:Default
      (CallFunction|ResetJumpCounter) ; Reset jump counter on other triggers \nsuch as swimming or ziplining.
    )
    (:Event.Jump.Pressed
      (bind _outfortplayerpawn (CallFunction|GetAvatarasValidFortPlayerPawn)
        (:ValidFortPlayerPawn
          (Variables|Hidden|SetPlayerRef _outfortplayerpawn)
          (if (CallFunction|JumpConditions)
            (CallFunction|JumpBootsonJump)
            (else
              ;[1 1 1] Failed Jump Logic
              (CallFunction|ResetJumpCounter))))))
    (:Event.Physics.Landed
      (CallFunction|JumpBootsonLand))))

(event Custom|JumpBoots_OnLand
  (bind (_jumpparams_ceilingheighttotrace_42_97759bbf46c015fd67dbaa9a9fd52bdc _jumpparams_launch_xyvelocitymultiplier_28_fd149de643cf2bf44ec729bb1b36d832 _jumpparams_launch_zvelocity_27_5bbf517844fc06657a6325b36dc3ade8 _jumpparams_launch_overridexyvelocity_30_64dc724b4d2faf89d63e1b81035d0ff8 _jumpparams_onjump_gameplayeffecttoapply_43_70fc651342557968aa109eba1d306a59 _jumpparams_onland_gameplayeffecttoapply_44_32a934db4255e53df4aaa78e2eb62d38) (Utilities|Struct|BreakSTRUCTJumpBootsJumpParams (CallFunction|GetJumpParamsforCurrentJump)))
  (bind (_jumpparams_ceilingheighttotrace_42_97759bbf46c015fd67dbaa9a9fd52bdc_2 _jumpparams_launch_xyvelocitymultiplier_28_fd149de643cf2bf44ec729bb1b36d832_2 _jumpparams_launch_zvelocity_27_5bbf517844fc06657a6325b36dc3ade8_2 _jumpparams_launch_overridexyvelocity_30_64dc724b4d2faf89d63e1b81035d0ff8_2 _jumpparams_onjump_gameplayeffecttoapply_43_70fc651342557968aa109eba1d306a59_2 _jumpparams_onland_gameplayeffecttoapply_44_32a934db4255e53df4aaa78e2eb62d38_2 _jumpparams_onjump_gameplaycue_48_0fec94334bbbb7b0c7fa2b930cd62862) (Utilities|Struct|BreakSTRUCTJumpBootsJumpParams (CallFunction|GetJumpParamsforCurrentJump)))
  (bind (_jumpparams_ceilingheighttotrace_42_97759bbf46c015fd67dbaa9a9fd52bdc_3 _jumpparams_launch_xyvelocitymultiplier_28_fd149de643cf2bf44ec729bb1b36d832_3 _jumpparams_launch_zvelocity_27_5bbf517844fc06657a6325b36dc3ade8_3 _jumpparams_launch_overridexyvelocity_30_64dc724b4d2faf89d63e1b81035d0ff8_3 _jumpparams_onjump_gameplayeffecttoapply_43_70fc651342557968aa109eba1d306a59_3 _jumpparams_onland_gameplayeffecttoapply_44_32a934db4255e53df4aaa78e2eb62d38_3 _jumpparams_onjump_gameplaycue_48_0fec94334bbbb7b0c7fa2b930cd62862_2 _jumpparams_onland_gameplaycue_50_a57dc3a948f564e8e76069a9b7ee0168) (Utilities|Struct|BreakSTRUCTJumpBootsJumpParams (CallFunction|GetJumpParamsforCurrentJump)))
  (if (>= (Variables|Default|GetJumpCounter) 0)
    (CallFunction|ValidateandExecuteGE _jumpparams_onland_gameplayeffecttoapply_44_32a934db4255e53df4aaa78e2eb62d38) ; Add GE's to make each jump unique.
    (CallFunction|ValidateandExecuteGCN _jumpparams_onland_gameplaycue_50_a57dc3a948f564e8e76069a9b7ee0168) ; Add GCN/GCNL's  in blueprint to prevent\nreplication delay.
    (CallFunction|ValidateandRemoveGCN _jumpparams_onjump_gameplaycue_48_0fec94334bbbb7b0c7fa2b930cd62862) ; Remove GCN/GCNL's here because they aren't\nattached to a GE for removal.
    (if (CallFunction|FinalJumpConditions)
      (CallFunction|ResetJumpCounter)
      (Ability|CommitAbility) ; Trigger CD.
      (Ability|EndAbility)
      (else
        (bind _returnvalue (Utilities|Time|SetTimerbyFunctionName self "ResetJumpCounter" (ScalableFloat|GetValueatLevel (Variables|Default|GetOnLanded_JumpResetTime)))) ; Set reactivation time that waits for another\njump from player.
        (Variables|Default|SetTimer_ResetJump _returnvalue)))))

(fn CheckCeilingHeight
  (bind _jumpcounter (Variables|Default|GetJumpCounter))
  (bind _returnvalue (+ _jumpcounter 1))
  (bind _jumpparams (Variables|Default|GetJumpParams))
  (bind _output_ceilingheighttotrace_42_97759bbf46c015fd67dbaa9a9fd52bdc (Utilities|Struct|BreakSTRUCTJumpBootsJumpParams (Utilities|Array|Get(acopy) _jumpparams _returnvalue)))
  (if (> _output_ceilingheighttotrace_42_97759bbf46c015fd67dbaa9a9fd52bdc 0.0)
    (bind (_outhit _returnvalue_2) (Collision|LineTraceByChannel (Transformation|GetWorldLocation (Variables|/Script/Engine.Character|GetMesh (Variables|Hidden|GetPlayerRef))) (+ (* "0, 0,1.000000" _output_ceilingheighttotrace_42_97759bbf46c015fd67dbaa9a9fd52bdc) (Transformation|GetWorldLocation (Variables|/Script/Engine.Character|GetMesh (Variables|Hidden|GetPlayerRef)))) "TraceTypeQuery10"))
    (return (not _returnvalue_2))
    (else
      (return true))))

(fn FinalJumpConditions
  (return (== (Variables|Default|GetJumpCounter) (- (Utilities|Array|Length (Variables|Default|GetJumpParams)) 1))))

(fn GetJumpParamsForCurrentJump
  (return (Utilities|Array|Get(acopy) (Variables|Default|GetJumpParams) (Variables|Default|GetJumpCounter))))

(fn GetPlayerAcceleration
  (params (mult :type Float%20%28double-precision%29))
  (bind (_returnvalue_x _returnvalue_y _returnvalue_z) (Math|Vector|BreakVector (Pawn|Components|CharacterMovement|GetCurrentAcceleration (Variables|/Script/Engine.Character|GetCharacterMovement (Variables|Hidden|GetPlayerRef)))))
  (return (* _returnvalue_x (Variables|Getmult)) (* _returnvalue_y (Variables|Getmult)) (* _returnvalue_z (Variables|Getmult))))

(fn JumpConditions
  (if (not (GameplayTags|HasAnyMatchingGameplayTags (Ability|GetAbilitySystemComponent (Variables|Hidden|GetPlayerRef)) (Variables|Default|GetActivationBlockingTags)))
    (bind _canjump (CallFunction|CheckCeilingHeight))
    (return _canjump)
    (else
      (return false))))

(fn LaunchPlayer
  (bind (_jumpparams_ceilingheighttotrace_42_97759bbf46c015fd67dbaa9a9fd52bdc _jumpparams_launch_xyvelocitymultiplier_28_fd149de643cf2bf44ec729bb1b36d832 _jumpparams_launch_zvelocity_27_5bbf517844fc06657a6325b36dc3ade8 _jumpparams_launch_overridexyvelocity_30_64dc724b4d2faf89d63e1b81035d0ff8) (Utilities|Struct|BreakSTRUCTJumpBootsJumpParams (CallFunction|GetJumpParamsforCurrentJump)))
  (bind (_x _y) (CallFunction|GetPlayerAcceleration (ScalableFloat|GetValueatLevel _jumpparams_launch_xyvelocitymultiplier_28_fd149de643cf2bf44ec729bb1b36d832 (Math|Conversions|ToFloat(Integer) (Variables|Default|GetJumpCounter)))))
  (Pawn|Character|LaunchCharacterJump (Variables|Hidden|GetPlayerRef) (Math|Vector|MakeVector _x _y (ScalableFloat|GetValueatLevel _jumpparams_launch_zvelocity_27_5bbf517844fc06657a6325b36dc3ade8 (Math|Conversions|ToFloat(Integer) (Variables|Default|GetJumpCounter)))) (Class|AthenaFunctionLibrary|RowtoBool _jumpparams_launch_overridexyvelocity_30_64dc724b4d2faf89d63e1b81035d0ff8 (Math|Conversions|ToFloat(Integer) (Variables|Default|GetJumpCounter))) true false false))

(fn ResetJumpCounter
  (Variables|Default|SetJumpCounter -1) ; Increments +1 on start of jump.\n\n0 == first jump.
  (Development|PrintString "JumpBoots: Jump Reset" (Variables|Default|GetDebug) (Variables|Default|GetDebug) "(R=0.000000,G=0.660000,B=1.000000,A=1.000000)" 5.0))

(fn UpdateJumpVars
  (Utilities|Time|ClearandInvalidateTimerbyHandle (Variables|Default|GetTimer_ResetJump)) ; Clear the "jump reactivation" timer as\nthis runs after a successful jump.
  (Math|Integer|IncrementInt (Variables|Default|GetJumpCounter)
    (:%20%20
      (Development|PrintString (Utilities|String|Append "JumpBoots: Jump Counter - " (Utilities|String|ToString(Integer) (Variables|Default|GetJumpCounter))) (Variables|Default|GetDebug) (Variables|Default|GetDebug) "(R=0.000000,G=0.660000,B=1.000000,A=1.000000)" 5.0))))

(fn ValidateAndExecuteGCN
  (params (GameplayCueTag :type Gameplay%20Cue%20Tag%20Structure%20%28by%20ref%29))
  (bind _gameplaycuetag (Utilities|Struct|BreakGameplayCueTag GameplayCueTag))
  (bind _jumpcounter (Variables|Default|GetJumpCounter))
  (bind _returnvalue (Math|Conversions|ToFloat(Integer) _jumpcounter))
  (bind _returnvalue_2 (Ability|GameplayCue|MakeGameplayCueParameters 0.0 _returnvalue))
  (if (GameplayTags|IsGameplayTagValid _gameplaycuetag)
    (Ability|ExecuteGameplayCueWithParamsOnOwner _gameplaycuetag _returnvalue_2)
    (Ability|AddGameplayCueWithParamsToOwner _gameplaycuetag _returnvalue_2)))

(fn ValidateAndExecuteGE
  (params (GameplayEffects :type Array%20of%20Gameplay%20Effect%20Class%20References))
  (for _array_element GameplayEffects
    (if (Utilities|IsValidClass _array_element)
      (GameplayEffects|ApplyGameplayEffectToTarget :self (Ability|GetAbilitySystemComponent (Variables|Hidden|GetPlayerRef)) :GameplayEffectClass _array_element :Target (Ability|GetAbilitySystemComponent (Variables|Hidden|GetPlayerRef)) :Level 0.0))))

(fn ValidateAndRemoveGCN
  (params (GameplayCueTag :type Gameplay%20Cue%20Tag%20Structure%20%28by%20ref%29))
  (bind _gameplaycuetag (Utilities|Struct|BreakGameplayCueTag GameplayCueTag))
  (if (GameplayTags|IsGameplayTagValid _gameplaycuetag)
    (Ability|RemoveGameplayCueFromOwner _gameplaycuetag)))
